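# Animation Blueprint for the drivable Ship Cannon vehicle, posing its body, wheels, cannon pivot and trigger.
# A has-driver pose or reference pose is layered with the TriggerPart state machine on the cannon_pivot and trigger branches.
# TriggerPart moves between loaded, fire, unloaded and reload animations using fired and cooldown flags.
# Bone-space Modify Bone nodes then apply per-wheel spin and cannon rotation from variables, with no event graph logic.
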AnimBlueprint /Game/Athena/DrivableVehicles/Ship_Cannon/Meshes/Ship_Cannon_VehAnimBP
target_skeleton=&/Ship_Cannon_Veh_Skeleton
graphs [
 AnimGraph {nodes [
   n1 = Blend Poses by bool(
    BlendPose_0=Sequence Player 'PushCannon_HasDriver_Pose_Vehicle'() {sequence=/Game/Animation/Game/MainPlayer/Vehicles/PushCannon/Vehicle/PushCannon_HasDriver_Pose_Vehicle}
    BlendPose_1=Local Space Ref Pose()
    BlendTime_0=0
    BlendTime_1=0.3
    bActiveValue=Get bHasDriver() -> (bHasDriver:bool)
   ) {blendType=Linear}
   AnimGraphNode_StateMachine_0 = TriggerPart() {is=title AnimGraphNode_StateMachine}
   n2 = "Transform (Modify) Bone - Bone: wheel_FR"(
    ComponentPose=Local To Component(LocalPose=Layered blend per bone(BasePose=@n1, BlendPoses_0=@AnimGraphNode_StateMachine_0) {layerSetup [{branchFilters [{boneName=cannon_pivot}, {boneName=trigger}]}]})
    Alpha=1
    Translation="0,0,0"
    Rotation=Get WheelSpinRotatorFR() -> (WheelSpinRotatorFR:Rotator)
    Scale="1,1,1"
   ) {rotationMode=BMM_Replace, rotationSpace=BCS_BoneSpace, alphaBoolBlend {blendOption=Linear}, boneToModify {boneName=wheel_FR}}
   n3 = "Transform (Modify) Bone - Bone: wheel_FL"(ComponentPose=@n2, Alpha=1, Translation="0,0,0", Rotation=Get WheelSpinRotatorFL() -> (WheelSpinRotatorFL:Rotator), Scale="1,1,1") {rotationMode=BMM_Replace, rotationSpace=BCS_BoneSpace, alphaBoolBlend {blendOption=Linear}, boneToModify {boneName=wheel_FL}}
   n4 = "Transform (Modify) Bone - Bone: wheel_BR"(ComponentPose=@n3, Alpha=1, Translation="0,0,0", Rotation=Get WheelSpinRotatorBR() -> (WheelSpinRotatorBR:Rotator), Scale="1,1,1") {rotationMode=BMM_Replace, rotationSpace=BCS_BoneSpace, alphaBoolBlend {blendOption=Linear}, boneToModify {boneName=wheel_BR}}
   n5 = "Transform (Modify) Bone - Bone: wheel_BL"(ComponentPose=@n4, Alpha=1, Translation="0,0,0", Rotation=Get WheelSpinRotatorBL() -> (WheelSpinRotatorBL:Rotator), Scale="1,1,1") {rotationMode=BMM_Replace, rotationSpace=BCS_BoneSpace, alphaBoolBlend {blendOption=Linear}, boneToModify {boneName=wheel_BL}}
   n6 = "Transform (Modify) Bone - Bone: cannon_pivot"!(ComponentPose=@n5, Alpha=1, Rotation=Get CannonLocalRotation() -> (CannonLocalRotation:Rotator)) -> (Pose) {rotationMode=BMM_Replace, rotationSpace=BCS_BoneSpace, alphaBoolBlend {blendOption=Linear}, boneToModify {boneName=cannon_pivot}}
   Output Pose(Result=Component To Local(ComponentPose=@n6)) {layerGroup=None}
  ]}
 TriggerPart {
  entry=Default
  path=AnimGraph.AnimGraphNode_StateMachine_0.TriggerPart
  ids=id
  nodes [
   AnimStateNode_0 = Default() {is=title AnimStateNode}
   AnimStateNode_1 = Fired() {is=title AnimStateNode}
   AnimStateNode_2 = FiredLoop() {is=title AnimStateNode}
   AnimStateNode_3 = Reload() {is=title AnimStateNode}
   AnimStateTransitionNode_0 = AnimStateTransitionNode() {crossfadeDuration=0}
   AnimStateTransitionNode_1 = AnimStateTransitionNode() {bAutomaticRuleBasedOnSequencePlayerInState=true, crossfadeDuration=0}
   AnimStateTransitionNode_2 = AnimStateTransitionNode() {crossfadeDuration=0}
   AnimStateTransitionNode_3 = AnimStateTransitionNode() {bAutomaticRuleBasedOnSequencePlayerInState=true, crossfadeDuration=0}
   AnimStateTransitionNode_4 = AnimStateTransitionNode() {crossfadeDuration=0}
   AnimStateTransitionNode_5 = AnimStateTransitionNode() {crossfadeDuration=0}
   AnimStateTransitionNode_6 = AnimStateTransitionNode() {crossfadeDuration=0}
  ]
  links [
   {
    dst=Fired
    rule=AnimGraph.AnimGraphNode_StateMachine_0.TriggerPart.AnimStateTransitionNode_0.Transition
    src=Default
    via=AnimStateTransitionNode_0
   }
   {
    dst=Fired
    rule=AnimGraph.AnimGraphNode_StateMachine_0.TriggerPart.AnimStateTransitionNode_4.Transition
    src=Default
    via=AnimStateTransitionNode_4
   }
   {
    dst=FiredLoop
    rule=AnimGraph.AnimGraphNode_StateMachine_0.TriggerPart.AnimStateTransitionNode_1.Transition
    src=Fired
    via=AnimStateTransitionNode_1
   }
   {
    dst=Default
    rule=AnimGraph.AnimGraphNode_StateMachine_0.TriggerPart.AnimStateTransitionNode_5.Transition
    src=FiredLoop
    via=AnimStateTransitionNode_5
   }
   {
    dst=Reload
    rule=AnimGraph.AnimGraphNode_StateMachine_0.TriggerPart.AnimStateTransitionNode_2.Transition
    src=FiredLoop
    via=AnimStateTransitionNode_2
   }
   {
    dst=Default
    rule=AnimGraph.AnimGraphNode_StateMachine_0.TriggerPart.AnimStateTransitionNode_3.Transition
    src=Reload
    via=AnimStateTransitionNode_3
   }
   {
    dst=Fired
    rule=AnimGraph.AnimGraphNode_StateMachine_0.TriggerPart.AnimStateTransitionNode_6.Transition
    src=Reload
    via=AnimStateTransitionNode_6
   }
  ]
 }
 Default {
  path=AnimGraph.AnimGraphNode_StateMachine_0.TriggerPart.AnimStateNode_0.Default
  nodes [Output Animation Pose(Result=Sequence Player 'PushCannon_Trigger_Loaded_Pose_Vehicle'() {sequence=/Game/Animation/Game/MainPlayer/Vehicles/PushCannon/Vehicle/PushCannon_Trigger_Loaded_Pose_Vehicle})]
 }
 Fired {
  path=AnimGraph.AnimGraphNode_StateMachine_0.TriggerPart.AnimStateNode_1.Fired
  nodes [Output Animation Pose(Result=Sequence Player 'PushCannon_Trigger_Fire_Vehicle'() {bLoopAnimation=false, sequence=/Game/Animation/Game/MainPlayer/Vehicles/PushCannon/Vehicle/PushCannon_Trigger_Fire_Vehicle})]
 }
 FiredLoop {
  path=AnimGraph.AnimGraphNode_StateMachine_0.TriggerPart.AnimStateNode_2.FiredLoop
  nodes [Output Animation Pose(Result=Sequence Player 'PushCannon_Trigger_Unloaded_Pose_Vehicle'() {sequence=/Game/Animation/Game/MainPlayer/Vehicles/PushCannon/Vehicle/PushCannon_Trigger_Unloaded_Pose_Vehicle})]
 }
 Reload {
  path=AnimGraph.AnimGraphNode_StateMachine_0.TriggerPart.AnimStateNode_3.Reload
  nodes [Output Animation Pose(Result=Sequence Player 'PushCannon_Trigger_Reload_Vehicle'() {bLoopAnimation=false, sequence=/Game/Animation/Game/MainPlayer/Vehicles/PushCannon/Vehicle/PushCannon_Trigger_Reload_Vehicle})]
 }
 Transition {
  path=AnimGraph.AnimGraphNode_StateMachine_0.TriggerPart.AnimStateTransitionNode_0.Transition
  nodes [Result(bCanEnterTransition=NOT Boolean(A=Get bIsCoolDownEnded() -> (bIsCoolDownEnded:bool)))]
 }
 Transition {path=AnimGraph.AnimGraphNode_StateMachine_0.TriggerPart.AnimStateTransitionNode_1.Transition, nodes [Result()]}
 Transition {
  path=AnimGraph.AnimGraphNode_StateMachine_0.TriggerPart.AnimStateTransitionNode_2.Transition
  nodes [Result(bCanEnterTransition=Get bIsCoolDownEnded())]
 }
 Transition {path=AnimGraph.AnimGraphNode_StateMachine_0.TriggerPart.AnimStateTransitionNode_3.Transition, nodes [Result()]}
 Transition {
  path=AnimGraph.AnimGraphNode_StateMachine_0.TriggerPart.AnimStateTransitionNode_4.Transition
  nodes [Result(bCanEnterTransition=Get bIsCannonFired() -> (bIsCannonFired:bool))]
 }
 Transition {
  path=AnimGraph.AnimGraphNode_StateMachine_0.TriggerPart.AnimStateTransitionNode_5.Transition
  nodes [Result(bCanEnterTransition=Get bIsCannonFired())]
 }
 Transition {
  path=AnimGraph.AnimGraphNode_StateMachine_0.TriggerPart.AnimStateTransitionNode_6.Transition
  nodes [Result(bCanEnterTransition=Get bIsCannonFired())]
 }
 EventGraph {links [], nodes []}
]
variables []
