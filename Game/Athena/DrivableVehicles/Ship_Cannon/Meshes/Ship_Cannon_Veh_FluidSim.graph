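# Physics asset for the drivable ship cannon vehicle, named as the variant for fluid simulation.
# It approximates the cannon with one simple shape per bone: boxes for the frame, handle and four wheels, and a capsule for the cannon pivot.
# The front wheel boxes are larger than the rear ones.
# It defines no constraints, so it supplies collision shapes without any joints between bodies.
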PhysicsAsset /Game/Athena/DrivableVehicles/Ship_Cannon/Meshes/Ship_Cannon_Veh_FluidSim
bodies {
 Frame [{shapeType=Box, extentX=163, extentY=80, extentZ=54, shapeName=Frame_box, center=<17.5832 0 28.0045>, rotation=<rot 10 0 -0.0002>}]
 cannon_pivot [{shapeType=Capsule, length=160, radius=25, center=<67.0837 -1.9444 0.6299>, rotation=<rot 90 0 0>}]
 handle [{shapeType=Box, extentX=60, extentY=85, extentZ=18, shapeName=handle_box, center=<-29.347 0 -13.8991>, rotation=<rot 30 -0.0001 -0.0001>}]
 wheel_BL [{shapeType=Box, extentX=23, extentY=10, extentZ=23, shapeName=wheel_BL_box, center=<0 -5 0>}]
 wheel_BR [{shapeType=Box, extentX=23, extentY=10, extentZ=23, shapeName=wheel_BR_box, center=<0 2.3651 0>}]
 wheel_FL [{shapeType=Box, extentX=30, extentY=10, extentZ=30, shapeName=wheel_FL_box, center=<0.7446 -4.5591 0>}]
 wheel_FR [{shapeType=Box, extentX=30, extentY=10, extentZ=30, shapeName=wheel_FR_box, center=<1.6005 2.4775 0>}]
}
constraints []
