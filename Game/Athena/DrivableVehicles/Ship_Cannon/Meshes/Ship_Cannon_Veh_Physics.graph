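# Physics asset for the drivable ship cannon vehicle, defining its simplified collision as a single body on the Frame bone.
# A low box covers the base, a tilted capsule named Cannon covers the barrel above it, and a short DriverCollision capsule sits toward the rear for the driver.
# It has no constraints, so all three shapes move together as one rigid body.
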PhysicsAsset /Game/Athena/DrivableVehicles/Ship_Cannon/Meshes/Ship_Cannon_Veh_Physics
bodies {Frame [
  {shapeType=Capsule, length=30, radius=50, shapeName=DriverCollision, center=<-94.8496 -0.0005 87.9437>}
  {shapeType=Capsule, length=122, radius=45, shapeName=Cannon, center=<25.6405 -0.0009 88.5055>, rotation=<rot -80 0 0>}
  {shapeType=Box, extentX=170, extentY=100, extentZ=56, center=<18.4184 -0.0007 31.5096>}
 ]}
constraints []
